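# Translucent unlit Niagara trail material for the PawJasmineGravity greet effect, drawing an animated, colored digital-glitch pattern.
# Time-driven flipbook UVs sample glitch textures per channel, with step-timed pixel-noise distortion and a particle-driven glitch intensity parameter.
# Four tint colors are layered by masks and scaled by particle color, alongside a center-glow gradient.
# Opacity fades at trail ends, by depth, distance and view angle; mobile skips the latter three.
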Material /BRCosmetics/Characters/Player/Male/Medium/Bodies/M_MED_PawJasmineGravity/FX/M_Character_PawJasmineGravityGreet_Trail
graph {
 properties {
  bUsedWithNiagaraMeshParticles=true
  bUsedWithNiagaraRibbons=true
  bUsedWithNiagaraSprites=true
  blendMode=BLEND_Translucent
  pixelDepthOffsetMode=PDOM_Legacy
  shadingModel=MSM_Unlit
  usedShadingModels=Unlit
 }
 nodes [
  n1 = ScalarParameter() {defaultValue=12, parameterName=Steps}
  Multiply(A=@n1, B=DynamicParameter() {paramNames [UOffset, VOffset, GlitchIntensity, Param4]}.GlitchIntensity)
  n2 = MaterialFunctionCall(Float3=NamedRerouteUsage()) {materialFunction=&BreakOutFloat3Components}
  n3 = LinearInterpolate(
   A=LinearInterpolate(A=Constant3Vector(), B=VectorParameter() {parameterName=Color_01, defaultValue=<rgba 0 1.5 2.2 1>}.RGB, Alpha=@n2.R)
   B=VectorParameter() {parameterName=Color_02, defaultValue=<rgba 2 0 3 1>}.RGB
   Alpha=@n2.B
  )
  n4 = VectorParameter() {parameterName=Color_03, defaultValue=<rgba 4 0 0.06 1>}
  n5 = Constant3Vector()
  n6 = VectorParameter() {parameterName=Color_04, defaultValue=<rgba 0.0046 0.5 2.1 1>}
  n7 = Add(A=NamedRerouteUsage(), B=NamedRerouteUsage())
  Multiply(
   A=Add(A=LinearInterpolate(A=@n3, B=@n4.RGB, Alpha=@n2.G), B=LinearInterpolate(A=@n5, B=@n6.RGB, Alpha=@n7))
   B=Multiply(A=ParticleColor().RGB, B=ScalarParameter() {defaultValue=3, parameterName=ColorMult})
  )
  n8 = ScalarParameter() {defaultValue=0.01, parameterName=GlitchSpeed}
  n9 = Time()
  n10 = Divide(A=Ceil(None=Multiply(A=@n9, B=@n1)), B=@n1)
  Multiply(A=@n8, B=@n10)
  n11 = Multiply(A=ComponentMask(None=TextureCoordinate() {vTiling=0}) {r=true}, B=ScalarParameter() {defaultValue=1, parameterName=PatternTileU})
  n12 = ScalarParameter() {parameterName=Pattern_PanU}
  n13 = ComponentMask(None=TextureCoordinate() {uTiling=0}) {g=true}
  n14 = ScalarParameter() {defaultValue=1, parameterName=PatternTileV}
  n15 = ComponentMask(None=Panner(Coordinate=Multiply(A=@n13, B=@n14), Time=ScalarParameter() {parameterName=Pattern_PanV}) {speedY=0.5}) {g=true}
  n16 = AppendVector(A=ComponentMask(None=Panner(Coordinate=@n11, Time=@n12)) {r=true}, B=@n15)
  n17 = Constant() {r=1}
  n18 = ScalarParameter() {defaultValue=8, parameterName=Columns}
  n19 = ScalarParameter() {defaultValue=8, parameterName=Rows}
  n20 = AppendVector(A=Divide(A=@n17, B=@n18), B=Divide(A=@n17, B=@n19))
  n21 = Multiply(A=Time(), B=ScalarParameter() {defaultValue=16, parameterName=FB_Speed})
  n22 = Floor(None=@n21)
  n23 = Divide(A=@n21, B=@n19)
  n24 = AppendVector(A=Divide(A=Fmod(A=@n22, B=@n19), B=@n19), B=Divide(A=Floor(None=@n23), B=@n18))
  n25 = DynamicParameter() {paramNames [UOffset, VOffset, GlitchIntensity, Param4]}
  n26 = Add(A=Add(A=Multiply(A=@n16, B=@n20), B=@n24), B=AppendVector(A=@n25.UOffset, B=@n25.UOffset))
  n27 = Add(A=@n8, B=@n10)
  n28 = TextureSample(UVs=Panner(Coordinate=TextureCoordinate() {uTiling=0.25, vTiling=0.25}, Time=@n27) {speedX=-1, speedY=-0.25}) {texture=&/T_FX_Character_PawJasmineGravityGreet_Glitch_Pixel_Alpha}
  n29 = TextureSample(UVs=Panner(Coordinate=TextureCoordinate() {uTiling=0.1, vTiling=-0.1}, Time=@n27) {speedX=-0.5, speedY=-0.5}) {texture=&T_FX_Character_PawJasmineGravityGreet_Glitch_Pixel_Alpha}
  n30 = AppendVector(A=@n28.R, B=@n29.G)
  n31 = Add(A=@n26, B=@n30)
  n32 = NamedRerouteUsage()
  n33 = Panner(Coordinate=LinearInterpolate(A=@n26, B=@n31, Alpha=@n32)) {speedX=-0.5}
  n34 = Multiply(A=ComponentMask(None=TextureCoordinate() {vTiling=0}) {r=true}, B=ScalarParameter() {defaultValue=1, parameterName=PatternTileU})
  n35 = ScalarParameter() {parameterName=Pattern_PanU}
  n36 = AppendVector(A=ComponentMask(None=Panner(Coordinate=@n34, Time=@n35) {speedX=0.5}) {r=true}, B=@n15)
  n37 = Constant() {r=1}
  n38 = AppendVector(
   A=Divide(A=@n37, B=ScalarParameter() {defaultValue=16, parameterName=ColumnsThin})
   B=Divide(A=@n37, B=ScalarParameter() {defaultValue=64, parameterName=RowsThin})
  )
  n39 = DynamicParameter() {paramNames [UOffset, VOffset, GlitchIntensity, Param4]}
  n40 = Add(A=Add(A=Multiply(A=@n36, B=@n38), B=@n24), B=AppendVector(A=@n39.UOffset, B=@n39.UOffset))
  n41 = Panner(Coordinate=LinearInterpolate(A=@n40, B=Add(A=@n40, B=@n30), Alpha=NamedRerouteUsage())) {speedX=-1}
  n42 = Panner(Coordinate=LinearInterpolate(A=@n26, B=@n31, Alpha=@n32)) {speedX=-0.25}
  NamedRerouteDeclaration(None=MaterialFunctionCall(
    X=TextureSample(UVs=@n33) {texture=&/T_FX_Character_PawJasmineGravityGreet_Glitch_01}.R
    Y=TextureSample(UVs=@n41) {texture=&T_FX_Character_PawJasmineGravityGreet_Glitch_01}.G
    Z=TextureSample(UVs=@n42) {texture=&T_FX_Character_PawJasmineGravityGreet_Glitch_01}.B
   ) {materialFunction=&MakeFloat3}.Result) {name=Glitch, nodeColor=<rgba 1 0.9375 0 1>}
  n43 = Multiply(A=ComponentMask(None=TextureCoordinate() {vTiling=0}) {r=true}, B=ScalarParameter() {defaultValue=1, parameterName=PatternTileU})
  n44 = ScalarParameter() {parameterName=Pattern_PanU}
  n45 = AppendVector(A=ComponentMask(None=Panner(Coordinate=@n43, Time=@n44) {speedX=0.5}) {r=true}, B=@n15)
  n46 = Constant() {r=1}
  n47 = AppendVector(
   A=Divide(A=@n46, B=ScalarParameter() {defaultValue=16, parameterName=ColumnsBase})
   B=Divide(A=@n46, B=ScalarParameter() {defaultValue=64, parameterName=RowsBase})
  )
  n48 = DynamicParameter() {paramNames [UOffset, VOffset, GlitchIntensity, Param4]}
  n49 = Add(A=Add(A=Multiply(A=@n45, B=@n47), B=@n24), B=AppendVector(A=@n48.UOffset, B=@n48.UOffset))
  n50 = Add(A=@n49, B=@n30)
  n51 = NamedRerouteUsage()
  n52 = TextureSample(UVs=Panner(Coordinate=LinearInterpolate(A=@n49, B=@n50, Alpha=@n51)) {speedX=-0.1}) {texture=&/T_FX_Character_PawJasmineGravityGreet_Glitch_FB_03_RGB}
  n53 = Desaturation(None=Add(A=NamedRerouteUsage(), B=@n52.G), Fraction=Constant() {r=1}) {luminanceFactors=<rgba 0.3 0.59 0.11 0>}
  n54 = NamedRerouteUsage()
  n55 = TextureCoordinate()
  n56 = ComponentMask(None=@n55) {r=true}
  n57 = Saturate(None=ConstantBiasScale(None=@n56) {bias=-0.025, scale=2})
  n58 = Saturate(None=ConstantBiasScale(None=OneMinus(None=@n56)) {bias=-0.025, scale=2})
  n59 = Multiply(A=Multiply(A=Add(A=@n53, B=@n54), B=ParticleColor().A), B=Saturate(None=Multiply(A=@n57, B=@n58)))
  n60 = ConstantBiasScale(None=Sine(None=Divide(A=@n56))) {bias=-0.75, scale=10}
  n61 = LinearInterpolate(A=@n59, B=Multiply(A=@n59, B=Saturate(None=@n60)), Alpha=ScalarParameter() {parameterName=UseOpacityForArms})
  n62 = ScalarParameter() {defaultValue=50, parameterName=Distance Divisor}
  n63 = DepthFade(
   Opacity=Multiply(
    A=@n61
    B=MaterialFunctionCall!(Distance Bias, Distance Divisor=@n62, Scene Depth, Pixel Depth) {materialFunction=/Game/Effects/Fort_Effects/Materials/_Functions/MF_DistanceBehindTranslucency_01}.Normalized Inverted
   )
   FadeDistance=ScalarParameter() {defaultValue=5, parameterName=DepthFadeDistance}
  )
  n64 = MaterialFunctionCall("FadeSensitivity (optional)"=ScalarParameter() {defaultValue=1, parameterName=AngleFadeSensitivity}) {materialFunction=&FadeBasedOnViewAngle}
  NamedRerouteDeclaration(None=ShadingPathSwitch(Default=Multiply(A=@n63, B=@n64.Result), Mobile=@n61) {desc=Migrated from FeatureLevelSwitch}) {name=Opacity, nodeColor=<rgba 1 0 0.9609 1>}
  NamedRerouteDeclaration(None=Multiply(A=@n52.G) {constB=5}) {name=GlitchBase, nodeColor=<rgba 0.7578 1 0 1>}
  NamedRerouteDeclaration(None=Multiply(
    A=ScalarParameter() {defaultValue=1, parameterName=GlitchIntensity}
    B=DynamicParameter() {paramNames [UOffset, VOffset, GlitchIntensity, Param4]}.GlitchIntensity
   )) {name=GlitchIntensity, nodeColor=<rgba 0.3828 0 1 1>}
  n65 = Sine(None=Multiply(A=ComponentMask(None=@n55) {g=true}) {constB=0.5})
  n66 = ScalarParameter() {defaultValue=0.4, parameterName=CenterGlowMult}
  NamedRerouteDeclaration(None=Power(Base=Multiply(A=@n65, B=@n66))) {name=Gradient, nodeColor=<rgba 0.0312 1 0 1>}
  NamedRerouteUsage()
  Sine(None=@n9)
 ]
}
