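# Niagara cosmetic effect for the PawJasmineGravity character body, named as its greet reactivity, that overlays flickering cylinders, ribbons and glitch particles on the skeletal mesh.
# Cylinder meshes follow bones or sockets, while ribbons trace sockets through BoneFlow; both last User.SystemDuration, boost brightness at night and collapse selected parts via foot/hand-hidden flags.
# GPU glitch sprites and decals sample the mesh with randomised variants or colours, alongside a CPU low variant and a disabled distance-debug emitter.
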NiagaraSystem /BRCosmetics/Characters/Player/Male/Medium/Bodies/M_MED_PawJasmineGravity/FX/NS_Character_PawJasmineGravityGreet_Reactivity
dependencies {
 usedDataInterfaces [
  &NiagaraDataInterfaceArrayFloat
  &NiagaraDataInterfaceArrayFloat3
  &NiagaraDataInterfaceArrayInt32
  &NiagaraDataInterfaceCurve
  &NiagaraDataInterfaceCamera
  &NiagaraDataInterfaceArrayFloat2
  &NiagaraDataInterfaceArrayColor
  &NiagaraDataInterfaceSkeletalMesh
 ]
 usedDynamicInputs []
 usedModules [
  &SystemState
  &self:SystemSpawnScript.SystemScriptSource.SystemScriptGraph.NiagaraNodeAssignment_1.SetVariables_#1
  &self:SystemSpawnScript.SystemScriptSource.SystemScriptGraph.NiagaraNodeAssignment_3.SetVariables_#2
 ]
}
emitters {
 Cylinders {
  emitterUpdateScript [
   &EmitterState()
   &SpawnBurst_Instantaneous(Spawn Count=4, Use Spawn Probability=false, UseLoopCountLimit=false)
  ]
  particleSpawnScript [
   &PartitionParticles(Number Of Partitions=4)
   &InitializeParticle(
    Color Mode=::NewEnumerator0
    Lifetime=$User.SystemDuration
    Mesh Scale=&Multiply_VectorByFloat(
     Vector=<0.35 1.1 1.1>
     Float=&SelectFloatFromArray(
      Direct Array Index=$Particles.PartitionParticles.Partition
      Float Selection Array=DataInterface() {propertyValues {FloatData [0.8,0.8,0.65,0.65]}}
     )
    )
    Mesh Scale Mode=::NewEnumerator4
    Sprite Size Mode=::NewEnumerator3
    UsePositionOffset=false
    Write Material Random=true
   )
   &InitialMeshOrientation(
    Mesh Orientation Mode=::NewEnumerator2
    Orientation Vector=&SelectVectorFromArray(
     Direct Array Index=$Particles.PartitionParticles.Partition
     Vector Selection Array=DataInterface() {propertyValues {FloatData [<0 0 -1>, <0 0 1>, <0 0 1>, <0 0 1>]}}
    )
    Rotation=&SelectVectorFromArray(
     Direct Array Index=$Particles.PartitionParticles.Partition
     Vector Selection Array=DataInterface() {propertyValues {FloatData [<0.5 0 0>, <0 0 0>, <0 0 -0.1>, <0 0 0.4>]}}
    )
    Use Rotation Vector=true
   )
   &InitialMeshOrientation(
    Mesh Orientation Mode=::NewEnumerator7
    Rotation=&SelectVectorFromArray(
     Direct Array Index=$Particles.PartitionParticles.Partition
     Vector Selection Array=DataInterface() {propertyValues {FloatData [<0 0 0>, <0 0 0>, <0 0 -0.05>, <0 0 0.4>]}}
    )
    Use Rotation Vector=true
   ) {module=InitialMeshOrientation002}
   &CameraOffset(Camera Offset Amount=&SelectFloatFromArray(Direct Array Index=$Particles.PartitionParticles.Partition, Float Selection Array=DataInterface() {propertyValues {FloatData [5,5,3,3]}}))
   /SetVariables(Particles.VisibilityTag=&SelectIntFromArray(Direct Array Index=$Particles.PartitionParticles.Partition, Int Selection Array=DataInterface() {propertyValues {IntData [0,0,1,1]}})) {isSetParameters=true}
  ]
  particleUpdateScript [
   &ParticleState()
   &ScaleMeshSize(Scale Factor=&VectorFromFloat(Value=&MakeCustomFloatFromBool(
      Bool=$User.bisFootHidden
      True Float=&SelectFloatFromArray(Direct Array Index=$Particles.PartitionParticles.Partition, Float Selection Array=DataInterface() {propertyValues {FloatData [0,0,1,1]}})
      False Float=1
     )))
   &ScaleMeshSize(Scale Factor=&VectorFromFloat(Value=&MakeCustomFloatFromBool(
      Bool=$User.bisHandHidden
      True Float=&SelectFloatFromArray(Direct Array Index=$Particles.PartitionParticles.Partition, Float Selection Array=DataInterface() {propertyValues {FloatData [1,1,0,0]}})
      False Float=1
     ))) {module=ScaleMeshSize001}
   &SkeletalMeshLocation(
    Bone / Socket Index=$Particles.PartitionParticles.Partition
    Bone Sampling Mode=::NewEnumerator4
    MaskSpawnGroup=false
    Sampled Position Offset=&SelectVectorFromArray(
     Direct Array Index=$Particles.PartitionParticles.Partition
     Vector Selection Array=DataInterface() {propertyValues {FloatData [<-13 0 0>, <21 0 0>, <8 0 0>, <-8 1 1>]}}
    )
    Skeletal Mesh=$System.Skeletal Mesh Ribbon
   )
   &ScaleColor(
    Scale Alpha=&Multiply_Float(
     A=&Multiply_Float(
      A=&FloatFromCurve(
       FloatCurve=DataInterface() {propertyValues {Curve {
          defaultValue=1
          keys [
           <at 0 0 RCIM_Cubic>
           {arriveTangentWeight=0.1389, interpMode=RCIM_Cubic, leaveTangentWeight=0.2778, tangentMode=RCTM_Break, time=0.25, value=1}
           {arriveTangentWeight=0.2778, interpMode=RCIM_Cubic, leaveTangentWeight=0.1389, tangentMode=RCTM_Break, time=0.75, value=1}
           <at 1 0 RCIM_Cubic>
          ]
         }}}
       Scale Curve=30
      )
      B=&Lerp_Float(
       A=&ClampFloat(Float=&Waveform("Waveform [1]"=::NewEnumerator8, "[1] Frequency"=20, "[1] Amplitude Min/Max"=<0 1>, "[1] Exponent"=2))
       B=1
       Alpha=&FloatFromCurve(FloatCurve=DataInterface() {propertyValues {Curve {
           defaultValue=1
           keys [
            <at 0 0 RCIM_Cubic>
            {arriveTangentWeight=0.1389, interpMode=RCIM_Cubic, leaveTangentWeight=0.2778, tangentMode=RCTM_Break, time=0.4, value=1}
            {arriveTangentWeight=0.2778, interpMode=RCIM_Cubic, leaveTangentWeight=0.1389, tangentMode=RCTM_Break, time=0.6, value=1}
            <at 1 0 RCIM_Cubic>
           ]
          }}})
      )
     )
     B=&Lerp_Float(
      A=0.25
      B=1
      Alpha=&Waveform("Waveform [1]"=::NewEnumerator8, "[1] Frequency"=8, "[1] Amplitude Min/Max"=<0 1>, "[1] Exponent"=2)
     )
    )
    Scale RGB=&FN_Select_Vector_from_Time_Of_Day(Morning=<1 1 1>, Day=<1 1 1>, Evening=<1 1 1>, Night=<3 3 3>)
    ScaleA=true
    ScaleRGB=true
   )
   &DynamicMaterialParameters(
    Index 0 Param 1=&RandomRangeFloat(Minimum=0, Maximum=1, Randomness Mode=::NewEnumerator1, Random Seed=$Particles.PartitionParticles.Partition, Fixed Random Seed=true)
    Index 0 Param 2=&RandomRangeFloat(Minimum=0, Maximum=1, Randomness Mode=::NewEnumerator1, Random Seed=$Particles.PartitionParticles.Partition, Fixed Random Seed=true)
    Index 0 Param 3=&Waveform("Waveform [1]"=::NewEnumerator8, "[1] Frequency"=2, "[1] Amplitude Min/Max"=<0 1>)
    Module.Index0Float0Write=true
    Module.Index0Float1Write=true
    Module.Index0Float2Write=true
    Module.Index0Float3Write=true
   )
  ]
  rendererClasses [&NiagaraMeshRendererProperties]
  renderers [
   {rendererClass=&NiagaraMeshRendererProperties, rendererIndex=0}
   {rendererClass=&NiagaraMeshRendererProperties, rendererIndex=1}
  ]
 }
 FN_Debug_Distance {
  bEnabled=false
  emitterUpdateScript [
   &EmitterState(Life Cycle Mode=::NewEnumerator1, Loop Behavior=::NewEnumerator1, Loop Duration Mode=::NewEnumerator1, UseLoopDelay=false)
   &SpawnBurst_Instantaneous(Spawn Count=1, Use Spawn Probability=false)
  ]
  particleSpawnScript [/Niagara/Modules/Spawn/Initialization/InitializeParticle(
    Lifetime=5
    Module.Write Color=true
    Module.Write Lifetime=true
    Module.Write Mass=true
    Module.Write Material Random=true
    Module.Write Position=true
    Module.Write Scale=false
    Module.Write SpriteRotation=false
    Module.Write SpriteSize=true
    Position=&RandomRangeVector(Minimum=<0 0 0>, Maximum=<50 50 50>)
    Sprite Size=<20 20>
   )]
  particleUpdateScript [
   &ParticleState(Kill Particles When Lifetime Has Elapsed=false, Let Infinitely Lived Particles Die When Emitter Deactivates=true)
   /SetVariables(Particles.Debug:NiagaraFloat=&Divide_Float(A=$Engine.Owner.LODDistance, B=10)) {isSetParameters=true}
   &DynamicMaterialParameters(
    Index 0 Param 1=$Particles.Debug
    Module.Param0WriteEnabled=true
    Module.Param1WriteEnabled=true
    Module.Param2WriteEnabled=true
    Module.Param3WriteEnabled=true
   )
   &CameraOffset(Camera Offset Amount=200)
  ]
  rendererClasses [&NiagaraSpriteRendererProperties]
  renderers [{rendererClass=&NiagaraSpriteRendererProperties}]
 }
 GlitchDecals {
  simTarget=GPUComputeSim
  emitterUpdateScript [
   &EmitterState()
   &SpawnRate(
    SpawnRate=&Multiply_Float(
     A=50
     B=&FloatFromCurve(
      FloatCurve=DataInterface() {propertyValues {Curve {
         defaultValue=1
         keys [
          {interpMode=RCIM_Cubic, tangentMode=RCTM_User, time=0, value=0}
          {arriveTangentWeight=0.1067, interpMode=RCIM_Cubic, leaveTangentWeight=0.3202, tangentMode=RCTM_User, time=0.2, value=1}
          {arriveTangentWeight=0.3202, interpMode=RCIM_Cubic, leaveTangentWeight=0.1067, tangentMode=RCTM_User, time=0.7, value=1}
          <at 1 0 RCIM_Cubic>
         ]
        }}}
      CurveIndex=$Emitter.NormalizedLoopAge
     )
    )
    Use Spawn Probability=true
   )
  ]
  particleSpawnScript [
   /SetVariables(Particles.ParticleVariant:NiagaraInt32=&MakeIntFromFloat(Float=&FloatFromCurve(
      FloatCurve=DataInterface() {propertyValues {Curve {keys [<at 0 0 RCIM_Constant>,<at 0.6 1 RCIM_Constant>,<at 0.8 2 RCIM_Constant>,<at 0.9 3 RCIM_Constant>,<at 0.95 4 RCIM_Constant>]}}}
      CurveIndex=&RandomRangeFloat(Minimum=0, Maximum=1)
     ))) {isSetParameters=true, module=SetVariables_#3}
   /SetVariables(Particles.Random=&RandomRangeFloat(Minimum=0, Maximum=1)) {isSetParameters=true}
   /Niagara/Modules/Spawn/Initialization/V2/InitializeParticle(
    Color Mode=::NewEnumerator0
    Lifetime=0.4
    Sprite Size Mode=::NewEnumerator3
    Uniform Sprite Size=8
    UsePositionOffset=false
    Write Material Random=true
   )
   &ShapeLocation(MaskSpawnGroup=false, Override Seed=false, Sphere Radius=3, UseOwnerScale=true)
   &CameraOffset(Camera Offset Amount=10)
  ]
  particleUpdateScript [
   &ParticleState()
   &SkeletalMeshLocation(MaskSpawnGroup=false, Mesh Sampling Type=::NewEnumerator5, OverrideRandomSeed=false, Skeletal Mesh=$System.Skeletal Mesh Ribbon)
   &Color()
   &ScaleColor(
    Color Value To Scale=$Particles.Color
    Scale Alpha=&FloatFromCurve(
     FloatCurve=DataInterface() {propertyValues {Curve {keys [
         <at 0 0 RCIM_Cubic>
         {arriveTangent=0.0087, interpMode=RCIM_Cubic, leaveTangent=0.0087, tangentMode=RCTM_User, time=0.15, value=1.0023}
         {arriveTangent=-1.1792, interpMode=RCIM_Cubic, leaveTangent=-1.1792, time=0.7, value=0.8}
         <at 1 0>
        ]}}}
     Scale Curve=0.5
    )
    Scale RGB=&MakeVectorFromLinearColorRGB(Color=&RandomRangeLinearColor(Minimum=<rgba 0.5938 0.6805 1 1>, Maximum=<rgba 1 0.5417 0.5422 1>))
    ScaleA=true
    ScaleRGB=true
   )
   &DynamicMaterialParameters(
    Index 0 Param 1=&Lerp_Float(A=0.2, B=1.1, Alpha=$Particles.Random)
    Index 0 Param 2=&Lerp_Float(A=1, B=2, Alpha=$Particles.Random)
    Index 0 Param 3=&Lerp_Float(A=0.3, B=1, Alpha=$Particles.Random)
    Index 0 Param 4=-1
    Module.Param0WriteEnabled=true
    Module.Param1WriteEnabled=true
    Module.Param2WriteEnabled=true
    Module.Param3WriteEnabled=true
   ) {module=DynamicMaterialParameters001}
  ]
  rendererClasses [&NiagaraSpriteRendererProperties]
  renderers [{rendererClass=&NiagaraSpriteRendererProperties}]
 }
 GlitchSprites {
  simTarget=GPUComputeSim
  emitterUpdateScript [
   &EmitterState()
   &SpawnRate(
    Spawn Probability=0.3
    SpawnRate=&Multiply_Float(
     A=50
     B=&FloatFromCurve(
      FloatCurve=DataInterface() {propertyValues {Curve {
         defaultValue=1
         keys [
          {interpMode=RCIM_Cubic, tangentMode=RCTM_User, time=0, value=0}
          {arriveTangentWeight=0.1067, interpMode=RCIM_Cubic, leaveTangentWeight=0.3202, tangentMode=RCTM_User, time=0.2, value=1}
          {arriveTangentWeight=0.3202, interpMode=RCIM_Cubic, leaveTangentWeight=0.1067, tangentMode=RCTM_User, time=0.7, value=1}
          <at 1 0 RCIM_Cubic>
         ]
        }}}
      CurveIndex=$Emitter.NormalizedLoopAge
     )
    )
    Use Spawn Probability=true
   )
  ]
  particleSpawnScript [
   /SetVariables(Particles.ParticleVariant:NiagaraInt32=&MakeIntFromFloat(Float=&FloatFromCurve(
      FloatCurve=DataInterface() {propertyValues {Curve {keys [<at 0 0 RCIM_Constant>,<at 0.6 1 RCIM_Constant>,<at 0.8 2 RCIM_Constant>,<at 0.9 3 RCIM_Constant>,<at 0.95 4 RCIM_Constant>]}}}
      CurveIndex=&RandomRangeFloat(Minimum=0, Maximum=1)
     ))) {isSetParameters=true, module=SetVariables_#3}
   &InitializeParticle(
    Color Mode=::NewEnumerator0
    Lifetime=0.2
    Sprite Size Mode=::NewEnumerator3
    Uniform Sprite Size=&MakeFloatFromVector2d(Vector 2D=&Multiply_Vector2D_ByFloat(Float=2, Vector2D=&RandomRangeVector2D(Minimum=<1.5 3>, Maximum=<3 5>)))
    UsePositionOffset=false
    Write Material Random=true
   )
   &ShapeLocation(MaskSpawnGroup=false, Override Seed=false, Sphere Radius=3, UseOwnerScale=true)
   &DynamicMaterialParameters(
    Index 0 Param 1=&RandomRangeFloat(Minimum=0.1, Maximum=0.3)
    Index 0 Param 2=&RandomRangeFloat(Minimum=1.5, Maximum=2.5)
    Index 0 Param 3=&RandomRangeFloat(Minimum=0.1, Maximum=0.5)
    Index 0 Param 4=1
    Module.Index0Float0Write=true
    Module.Index0Float1Write=true
    Module.Index0Float2Write=true
    Module.Index0Float3Write=true
   )
   &CameraOffset(Camera Offset Amount=5)
  ]
  particleUpdateScript [
   &ParticleState()
   &SkeletalMeshLocation(MaskSpawnGroup=false, Mesh Sampling Type=::NewEnumerator5, OverrideRandomSeed=false, Skeletal Mesh=$System.Skeletal Mesh Ribbon)
   &ScaleSpriteSize(
    Scale Factor=&Multiply_Vector2D(
     A=&SelectVector2DFromArray(
      Direct Array Index=$Particles.ParticleVariant
      Vector 2D Selection Array=DataInterface() {propertyValues {FloatData [<1 1>, <1 1>, <0.8 1.2>, <0.4 1.5>, <0.2 2>]}}
     )
     B=&MakeVector2D(
      X=&FloatFromCurve(FloatCurve=DataInterface() {propertyValues {Curve {keys [<at 0.5 1>,<at 1 0>]}}})
      Y=&FloatFromCurve(FloatCurve=DataInterface() {propertyValues {Curve {keys [<at 0 1>,<at 1 2>]}}})
     )
    )
    Scale Sprite Size Mode=::NewEnumerator2
   )
   &Color(Color=&SelectLinearColorFromArray(
     Array Sampling Mode=::NewEnumerator1
     Direct Array Index=$Particles.ParticleVariant
     Color Selection Array=DataInterface() {propertyValues {ColorData [<rgba 0.0087 0.6121 1 1>, <rgba 0 0.0336 1 1>, <rgba 0.4623 0 1 1>, <rgba 1 0 0.0327 1>, <rgba 1 0.7693 0 1>]}}
     Random Seed=?
    ))
   &ScaleColor(
    Color Value To Scale=$Particles.Color
    Scale Alpha=&FloatFromCurve(
     FloatCurve=DataInterface() {propertyValues {Curve {keys [
         <at 0 0 RCIM_Cubic>
         {arriveTangent=0.0087, interpMode=RCIM_Cubic, leaveTangent=0.0087, tangentMode=RCTM_User, time=0.15, value=1.0023}
         {arriveTangent=-1.1792, interpMode=RCIM_Cubic, leaveTangent=-1.1792, time=0.7, value=0.8}
         <at 1 0>
        ]}}}
     Scale Curve=2
    )
    Scale RGB=&MakeVectorFromLinearColorRGB(Color=&RandomRangeLinearColor(Minimum=<rgba 0.5938 0.6805 1 1>, Maximum=<rgba 1 0.5417 0.5422 1>))
    ScaleA=true
    ScaleRGB=true
   )
  ]
  rendererClasses [&NiagaraSpriteRendererProperties]
  renderers [{rendererClass=&NiagaraSpriteRendererProperties}]
 }
 GlitchSprites_LOW {
  emitterUpdateScript [
   &EmitterState()
   &SpawnRate(
    Spawn Probability=0.3
    SpawnRate=&FN_FrontEndFloatModifier(Gameplay=500, Front End=0)
    Use Spawn Probability=true
   )
  ]
  particleSpawnScript [
   /SetVariables(Particles.ParticleVariant:NiagaraInt32=&MakeIntFromFloat(Float=&FloatFromCurve(
      FloatCurve=DataInterface() {propertyValues {Curve {keys [<at 0 0 RCIM_Constant>,<at 0.6 1 RCIM_Constant>,<at 0.8 2 RCIM_Constant>,<at 0.9 3 RCIM_Constant>,<at 0.95 4 RCIM_Constant>]}}}
      CurveIndex=&RandomRangeFloat(Minimum=0, Maximum=1)
     ))) {isSetParameters=true, module=SetVariables_#3}
   &InitializeParticle(
    Color Mode=::NewEnumerator0
    Lifetime=0.3
    Sprite Size Mode=::NewEnumerator3
    Uniform Sprite Size=&MakeFloatFromVector2d(Vector 2D=&Multiply_Vector2D_ByFloat(Float=2.5, Vector2D=&RandomRangeVector2D(Minimum=<1.5 3>, Maximum=<3 5>)))
    UsePositionOffset=false
    Write Material Random=true
   )
   &ShapeLocation(MaskSpawnGroup=false, Override Seed=false, Sphere Radius=3, UseOwnerScale=true)
   &DynamicMaterialParameters(
    Index 0 Param 1=&RandomRangeFloat(Minimum=0.1, Maximum=0.3)
    Index 0 Param 2=&RandomRangeFloat(Minimum=1.5, Maximum=2.5)
    Index 0 Param 3=&RandomRangeFloat(Minimum=0.1, Maximum=0.5)
    Index 0 Param 4=1
    Module.Index0Float0Write=true
    Module.Index0Float1Write=true
    Module.Index0Float2Write=true
    Module.Index0Float3Write=true
   )
   &CameraOffset(Camera Offset Amount=10)
  ]
  particleUpdateScript [
   &ParticleState()
   &SkeletalMeshLocation(MaskSpawnGroup=false, Mesh Sampling Type=::NewEnumerator5, OverrideRandomSeed=false, Skeletal Mesh=$System.Skeletal Mesh Ribbon)
   &ScaleSpriteSize(
    Scale Factor=&Multiply_Vector2D(
     A=&SelectVector2DFromArray(
      Direct Array Index=$Particles.ParticleVariant
      Vector 2D Selection Array=DataInterface() {propertyValues {FloatData [<1 1>, <1 1>, <0.8 1.2>, <0.4 1.5>, <0.2 2>]}}
     )
     B=&MakeVector2D(
      X=&FloatFromCurve(FloatCurve=DataInterface() {propertyValues {Curve {keys [<at 0.5 1>,<at 1 0>]}}})
      Y=&FloatFromCurve(FloatCurve=DataInterface() {propertyValues {Curve {keys [<at 0 1>,<at 1 2>]}}})
     )
    )
    Scale Sprite Size Mode=::NewEnumerator2
   )
   &Color(Color=&SelectLinearColorFromArray(
     Array Sampling Mode=::NewEnumerator1
     Direct Array Index=$Particles.ParticleVariant
     Color Selection Array=DataInterface() {propertyValues {ColorData [<rgba 0.0087 0.6121 1 1>, <rgba 0 0.0336 1 1>, <rgba 0.4623 0 1 1>, <rgba 1 0 0.0327 1>, <rgba 1 0.7693 0 1>]}}
     Random Seed=?
    ))
   &ScaleColor(
    Color Value To Scale=$Particles.Color
    Scale Alpha=&FloatFromCurve(
     FloatCurve=DataInterface() {propertyValues {Curve {keys [
         <at 0 0 RCIM_Cubic>
         {arriveTangent=0.0087, interpMode=RCIM_Cubic, leaveTangent=0.0087, tangentMode=RCTM_User, time=0.15, value=1.0023}
         {arriveTangent=-1.1792, interpMode=RCIM_Cubic, leaveTangent=-1.1792, time=0.7, value=0.8}
         <at 1 0>
        ]}}}
     Scale Curve=4
    )
    Scale RGB=&Multiply_Vector(
     A=&MakeVectorFromLinearColorRGB(Color=&RandomRangeLinearColor(Minimum=<rgba 0.5938 0.6805 1 1>, Maximum=<rgba 1 0.5417 0.5422 1>))
     B=<3 3 3>
    )
    ScaleA=true
    ScaleRGB=true
   )
  ]
  rendererClasses [&NiagaraSpriteRendererProperties]
  renderers [{rendererClass=&NiagaraSpriteRendererProperties}]
 }
 Ribbons {
  emitterUpdateScript [
   /SetVariables(Emitter.CountPerRibbon=6) {isSetParameters=true}
   &EmitterState()
   &SpawnBurst_Instantaneous(
    Spawn Count=&Multiply_Int(A=$Emitter.CountPerRibbon, B=33)
    Use Spawn Probability=false
    UseLoopCountLimit=false
   )
  ]
  particleSpawnScript [
   &FN_SetSpawnGroup(
    Inclusivity=true
    Normalize Spawn Group=true
    Particles Per Group:NiagaraInt32=$Emitter.CountPerRibbon
    Pattern=::NewEnumerator0
    Set Ribbon ID and Ribbon Link Order=true
   )
   &InitializeParticle(
    Lifetime=$User.SystemDuration
    Ribbon Width=3.5
    Ribbon Width Mode=::NewEnumerator1
    Sprite Size Mode=::NewEnumerator3
    UsePositionOffset=false
    Write Material Random=true
   )
   &CameraOffset()
   /Game/Effects/Niagara/Modules/BoneFlow/FN_BoneFlow_v4/FN_BoneFlow(
    BoneID:NiagaraInt32=$Particles.SpawnGroup
    Height:NiagaraFloat=$Particles.Normalized Spawn Group
    SkeletalMesh:NiagaraDataInterfaceSkeletalMesh=$System.Skeletal Mesh Ribbon
    Use Sockets Instead=true
    Write Position Attribute=true
   )
  ]
  particleUpdateScript [
   &ParticleState()
   &FN_BoneFlow(
    BoneID:NiagaraInt32=$Particles.SpawnGroup
    Height:NiagaraFloat=$Particles.Normalized Spawn Group
    SkeletalMesh:NiagaraDataInterfaceSkeletalMesh=$System.Skeletal Mesh Ribbon
    Use Sockets Instead=true
    Write Position Attribute=true
   ) {module=FN_BoneFlow001}
   /SetVariables(Particles.PositionOffset:NiagaraPosition=&AddVectorToPosition(
     Position=$Particles.Position
     Vector=&Multiply_VectorByFloat(
      Vector=&ReturnCameraVector(
       Camera Vectors=::NewEnumerator4
       Coordinate Space=::Simulation
       Origin Position=$Particles.Position
       CameraQuery=DataInterface() {propertyValues {bRequireCurrentFrameData=false}}
      )
      Float=&FN_FrontEndFloatModifier(Gameplay=10, Front End=0)
     )
    )) {isSetParameters=true}
   &ScaleColor(
    Scale Alpha=&Multiply_Float(
     A=&Multiply_Float(
      A=&Multiply_Float(
       A=&FloatFromCurve(
        FloatCurve=DataInterface() {propertyValues {Curve {
           defaultValue=1
           keys [
            <at 0 0 RCIM_Cubic>
            {arriveTangentWeight=0.1389, interpMode=RCIM_Cubic, leaveTangentWeight=0.2778, tangentMode=RCTM_Break, time=0.25, value=1}
            {arriveTangentWeight=0.2778, interpMode=RCIM_Cubic, leaveTangentWeight=0.1389, tangentMode=RCTM_Break, time=0.75, value=1}
            <at 1 0 RCIM_Cubic>
           ]
          }}}
        Scale Curve=30
       )
       B=&Lerp_Float(
        A=&ClampFloat(Float=&Waveform("Waveform [1]"=::NewEnumerator8, "[1] Frequency"=20, "[1] Amplitude Min/Max"=<0 1>, "[1] Exponent"=2))
        B=1
        Alpha=&FloatFromCurve(FloatCurve=DataInterface() {propertyValues {Curve {
            defaultValue=1
            keys [
             <at 0 0 RCIM_Cubic>
             {arriveTangentWeight=0.1389, interpMode=RCIM_Cubic, leaveTangentWeight=0.2778, tangentMode=RCTM_Break, time=0.4, value=1}
             {arriveTangentWeight=0.2778, interpMode=RCIM_Cubic, leaveTangentWeight=0.1389, tangentMode=RCTM_Break, time=0.6, value=1}
             <at 1 0 RCIM_Cubic>
            ]
           }}})
       )
      )
      B=&Lerp_Float(
       A=0.25
       B=1
       Alpha=&Waveform("Waveform [1]"=::NewEnumerator8, "[1] Frequency"=8, "[1] Amplitude Min/Max"=<0 1>, "[1] Exponent"=2)
      )
     )
     B=&FN_Select_Float_from_Time_Of_Day(Morning=1, Day=1, Evening=1, Night=1)
    )
    Scale RGB=&FN_Select_Vector_from_Time_Of_Day(Morning=<1 1 1>, Day=<1 1 1>, Evening=<1 1 1>, Night=<1.5 1.5 1.5>)
    ScaleA=true
    ScaleRGB=true
   )
   &DynamicMaterialParameters(
    Index 0 Param 1=&RandomRangeFloat(Minimum=0, Maximum=1, Randomness Mode=::NewEnumerator1, Random Seed=$Particles.SpawnGroup, Fixed Random Seed=true)
    Index 0 Param 2=&RandomRangeFloat(Minimum=0, Maximum=1, Randomness Mode=::NewEnumerator1, Random Seed=$Particles.SpawnGroup, Fixed Random Seed=true)
    Index 0 Param 3=&Waveform("Waveform [1]"=::NewEnumerator8, "[1] Frequency"=2, "[1] Amplitude Min/Max"=<0 1>, "[1] Random Seed"=$Particles.Initial.SpawnGroup)
    Module.Index0Float0Write=true
    Module.Index0Float1Write=true
    Module.Index0Float2Write=true
    Module.Index0Float3Write=true
   )
   &ScaleRibbonWidth(Ribbon Width Scale=&MakeCustomFloatFromBool(Bool=$User.bisLowLOD, True Float=0, False Float=1)) {enabled=false}
   &ScaleRibbonWidth(Ribbon Width Scale=&MakeCustomFloatFromBool(
     Bool=$User.bisFootHidden
     True Float=&SelectFloatFromArray(Direct Array Index=$Particles.SpawnGroup, Float Selection Array=DataInterface() {propertyValues {FloatData [1,0,1,0,1,0,1,0,1]}})
     False Float=1
    )) {module=ScaleRibbonWidth001}
   &ScaleRibbonWidth(Ribbon Width Scale=&MakeCustomFloatFromBool(
     Bool=$User.bisHandHidden
     True Float=&SelectFloatFromArray(
      Direct Array Index=$Particles.SpawnGroup
      Float Selection Array=DataInterface() {propertyValues {FloatData [1,1,1,1,1,1,1,1,1,1,1,1,1,1,1,1,1,1,1,1,1,1,0,0,1,0,0,1,0,0,1,0,0]}}
     )
     False Float=1
    )) {module=ScaleRibbonWidth002}
  ]
  rendererClasses [&NiagaraRibbonRendererProperties, &NiagaraSpriteRendererProperties]
  renderers [
   {rendererClass=&NiagaraRibbonRendererProperties}
   {rendererClass=&NiagaraSpriteRendererProperties, rendererIndex=1}
  ]
 }
}
summary {
 systemName=NS_Character_PawJasmineGravityGreet_Reactivity
 emitters [
  {emitterName=Cylinders, rendererClasses [&NiagaraMeshRendererProperties]}
  {emitterName=Ribbons, rendererClasses [&NiagaraRibbonRendererProperties, &NiagaraSpriteRendererProperties]}
  {emitterName=GlitchSprites, simTarget=GPUComputeSim, rendererClasses [&NiagaraSpriteRendererProperties]}
  {emitterName=GlitchDecals, simTarget=GPUComputeSim, rendererClasses [&NiagaraSpriteRendererProperties]}
  {bEnabled=false, emitterName=FN_Debug_Distance, rendererClasses [&NiagaraSpriteRendererProperties]}
  {emitterName=GlitchSprites_LOW, rendererClasses [&NiagaraSpriteRendererProperties]}
 ]
 userVariables [
  User.bisFootHidden {type=NiagaraBool}
  User.bIsFrontEnd {type=NiagaraBool}
  User.bIsFrontEndPreview {type=NiagaraBool}
  User.bisHandHidden {type=NiagaraBool}
  User.bisLowLOD {type=NiagaraBool}
  User.SystemDuration {type=NiagaraFloat, defaultValue=3, description=""}
 ]
}
