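# Niagara system for the PawJasmineGravity character's greet reactivity, emitting brief glitch sprites around the face accessory.
# One emitter spawns intermittently on a rise-and-fall rate curve, partitioning particles between face-mesh bones or sockets with mirrored offsets and velocities.
# A weighted random variant from a stepped curve picks each sprite's color and stretch, with randomized dynamic material parameters.
# User.bisHeadHidden scales the sprites to zero.
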NiagaraSystem /BRCosmetics/Characters/Player/Male/Medium/Bodies/M_MED_PawJasmineGravity/FX/NS_Character_PawJasmineGravityGreet_Reactivity_FaceAcc
dependencies {
 usedDataInterfaces [
  &NiagaraDataInterfaceCurve
  &NiagaraDataInterfaceArrayFloat3
  &NiagaraDataInterfaceArrayFloat2
  &NiagaraDataInterfaceArrayColor
  &NiagaraDataInterfaceSkeletalMesh
 ]
 usedDynamicInputs []
 usedModules [
  &SystemState
  &self:SystemSpawnScript.SystemScriptSource.SystemScriptGraph.NiagaraNodeAssignment_3.SetVariables_#1
 ]
}
emitters {Glitch {
  emitterUpdateScript [
   &EmitterState()
   &SpawnRate(
    Spawn Probability=0.3
    SpawnRate=&Multiply_Float(
     A=100
     B=&FloatFromCurve(
      FloatCurve=DataInterface() {propertyValues {Curve {
         defaultValue=1
         keys [
          {interpMode=RCIM_Cubic, tangentMode=RCTM_User, time=0, value=0}
          {arriveTangentWeight=0.1067, interpMode=RCIM_Cubic, leaveTangentWeight=0.3202, tangentMode=RCTM_User, time=0.2, value=1}
          {arriveTangentWeight=0.3202, interpMode=RCIM_Cubic, leaveTangentWeight=0.1067, tangentMode=RCTM_User, time=0.7, value=1}
          <at 1 0 RCIM_Cubic>
         ]
        }}}
      CurveIndex=$Emitter.NormalizedLoopAge
     )
    )
    Use Spawn Probability=true
   )
  ]
  particleSpawnScript [
   /SetVariables(Particles.ParticleVariant:NiagaraInt32=&MakeIntFromFloat(Float=&FloatFromCurve(
      FloatCurve=DataInterface() {propertyValues {Curve {keys [<at 0 0 RCIM_Constant>,<at 0.6 1 RCIM_Constant>,<at 0.8 2 RCIM_Constant>,<at 0.9 3 RCIM_Constant>,<at 0.95 4 RCIM_Constant>]}}}
      CurveIndex=&RandomRangeFloat(Minimum=0, Maximum=1)
     ))) {isSetParameters=true}
   &PartitionParticles()
   &InitializeParticle(
    Color Mode=::NewEnumerator0
    Lifetime=0.3
    Sprite Size Mode=::NewEnumerator3
    Uniform Sprite Size=&MakeFloatFromVector2d(Vector 2D=&Multiply_Vector2D_ByFloat(Float=1, Vector2D=&RandomRangeVector2D(Minimum=<1.5 4>, Maximum=<3 6>)))
    UsePositionOffset=false
    Write Material Random=true
   )
   &SkeletalMeshLocation(
    Bone / Socket Index=$Particles.PartitionParticles.Partition
    MaskSpawnGroup=false
    Mesh Sampling Type=::NewEnumerator5
    Sampled Position Offset=&SelectVectorFromArray(
     Direct Array Index=$Particles.PartitionParticles.Partition
     Vector Selection Array=DataInterface() {propertyValues {FloatData [<0 1 0.2>, <0 1 -0.2>]}}
    )
    Skeletal Mesh=$System.Skeletal Mesh Face
    Socket Sampling Mode=::NewEnumerator5
   )
   &ShapeLocation(MaskSpawnGroup=false, Override Seed=false, Sphere Radius=3, UseOwnerScale=true)
   &DynamicMaterialParameters(
    Index 0 Param 1=&RandomRangeFloat(Minimum=0.1, Maximum=0.3)
    Index 0 Param 2=&RandomRangeFloat(Minimum=1.5, Maximum=2.5)
    Index 0 Param 3=&RandomRangeFloat(Minimum=0.1, Maximum=0.5)
    Index 0 Param 4=1
    Module.Index0Float0Write=true
    Module.Index0Float1Write=true
    Module.Index0Float2Write=true
    Module.Index0Float3Write=true
   )
   &CameraOffset(Camera Offset Amount=-2)
   &AddVelocity(Velocity=&RotateVectorByQuaternion(
     Vector To Rotate=&SelectVectorFromArray(
      Direct Array Index=$Particles.PartitionParticles.Partition
      Vector Selection Array=DataInterface() {propertyValues {FloatData [<0 0 -1>, <0 0 1>]}}
     )
     Quaternion=$Particles.SkeletalMeshLocation.SampledRotation
    ))
   &ApplyInitialForces()
  ]
  particleUpdateScript [
   &ParticleState()
   &ScaleSpriteSize(
    Scale Factor=&Multiply_Vector2D(
     A=&SelectVector2DFromArray(
      Direct Array Index=$Particles.ParticleVariant
      Vector 2D Selection Array=DataInterface() {propertyValues {FloatData [<1 1>, <1 1>, <0.8 1.2>, <0.4 1.5>, <0.2 2>]}}
     )
     B=&MakeVector2D(
      X=&FloatFromCurve(FloatCurve=DataInterface() {propertyValues {Curve {keys [<at 0.5 1>,<at 1 0>]}}})
      Y=&FloatFromCurve(FloatCurve=DataInterface() {propertyValues {Curve {keys [<at 0 1>,<at 1 2>]}}})
     )
    )
    Scale Sprite Size Mode=::NewEnumerator2
   )
   &ScaleSpriteSize(
    Scale Sprite Size Mode=::NewEnumerator0
    Uniform Scale Factor=&MakeCustomFloatFromBool(Bool=$User.bisHeadHidden, True Float=0, False Float=1)
   ) {module=ScaleSpriteSize001}
   &Color(Color=&SelectLinearColorFromArray(
     Array Sampling Mode=::NewEnumerator1
     Direct Array Index=$Particles.ParticleVariant
     Color Selection Array=DataInterface() {propertyValues {ColorData [<rgba 0.0087 0.6121 1 1>, <rgba 0 0.0336 1 1>, <rgba 0.4623 0 1 1>, <rgba 1 0 0.0327 1>, <rgba 1 0.7693 0 1>]}}
     Random Seed=?
    ))
   &ScaleColor(
    Color Value To Scale=$Particles.Color
    Scale Alpha=&FloatFromCurve(
     FloatCurve=DataInterface() {propertyValues {Curve {keys [
         <at 0 0 RCIM_Cubic>
         {arriveTangent=0.0087, interpMode=RCIM_Cubic, leaveTangent=0.0087, tangentMode=RCTM_User, time=0.15, value=1.0023}
         {arriveTangent=-1.1792, interpMode=RCIM_Cubic, leaveTangent=-1.1792, time=0.7, value=0.8}
         <at 1 0>
        ]}}}
     Scale Curve=10
    )
    Scale RGB=&Multiply_VectorByFloat(
     Vector=&MakeVectorFromLinearColorRGB(Color=&RandomRangeLinearColor(Minimum=<rgba 0.5938 0.6805 1 1>, Maximum=<rgba 1 0.5417 0.5422 1>))
     Float=2
    )
    ScaleA=true
    ScaleRGB=true
   )
   &SkeletalMeshLocation(
    Bone / Socket Index=$Particles.PartitionParticles.Partition
    MaskSpawnGroup=false
    Mesh Sampling Type=::NewEnumerator5
    Orientation Sampling=::NewEnumerator1
    Sampled Position Offset=&SelectVectorFromArray(
     Direct Array Index=$Particles.PartitionParticles.Partition
     Vector Selection Array=DataInterface() {propertyValues {FloatData [<0 1 0.2>, <0 1 -0.2>]}}
    )
    Skeletal Mesh=$System.Skeletal Mesh Face
    Socket Sampling Mode=::NewEnumerator5
   ) {module=SkeletalMeshLocation001}
  ]
  rendererClasses [&NiagaraSpriteRendererProperties]
  renderers [{rendererClass=&NiagaraSpriteRendererProperties}]
 }}
summary {
 systemName=NS_Character_PawJasmineGravityGreet_Reactivity_FaceAcc
 emitters [{emitterName=Glitch, rendererClasses [&NiagaraSpriteRendererProperties]}]
 userVariables [
  User.bisHeadHidden {type=NiagaraBool}
  User.SystemDuration {type=NiagaraFloat, defaultValue=3, description=""}
 ]
}
