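# Niagara cosmetic effect for the PawJasmineGravity character's hood-down greet reactivity, adding glitch sprites and flickering ribbons around the face.
# GPU sprites spawn at two offset face-mesh bones or sockets, with weighted variants setting color and stretch; a LOW emitter uses constant-rate, larger, brighter sprites.
# Two BoneFlow ribbons follow sockets for User.SystemDuration, with waveform flicker, camera-vector offset and a night-time brightness boost.
# User.bisHeadHidden collapses all sprite sizes and ribbon widths to zero.
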NiagaraSystem /BRCosmetics/Characters/Player/Male/Medium/Bodies/M_MED_PawJasmineGravity/FX/NS_Character_PawJasmineGravityGreet_Reactivity_HoodDown
dependencies {
 usedDataInterfaces [
  &NiagaraDataInterfaceCurve
  &NiagaraDataInterfaceArrayFloat3
  &NiagaraDataInterfaceArrayFloat2
  &NiagaraDataInterfaceArrayColor
  &NiagaraDataInterfaceCamera
  &NiagaraDataInterfaceSkeletalMesh
 ]
 usedDynamicInputs [&MakeFloatFromVector2d]
 usedModules [
  &SystemState
  &self:SystemSpawnScript.SystemScriptSource.SystemScriptGraph.NiagaraNodeAssignment_3.SetVariables_#1
  &self:SystemSpawnScript.SystemScriptSource.SystemScriptGraph.NiagaraNodeAssignment_4.SetVariables_#2
 ]
}
emitters {
 GlitchSprites {
  simTarget=GPUComputeSim
  emitterUpdateScript [
   &EmitterState()
   &SpawnRate(
    Spawn Probability=0.3
    SpawnRate=&Multiply_Float(
     A=60
     B=&FloatFromCurve(
      FloatCurve=DataInterface() {propertyValues {Curve {
         defaultValue=1
         keys [
          {interpMode=RCIM_Cubic, tangentMode=RCTM_User, time=0, value=0}
          {arriveTangentWeight=0.1067, interpMode=RCIM_Cubic, leaveTangentWeight=0.3202, tangentMode=RCTM_User, time=0.2, value=1}
          {arriveTangentWeight=0.3202, interpMode=RCIM_Cubic, leaveTangentWeight=0.1067, tangentMode=RCTM_User, time=0.7, value=1}
          <at 1 0 RCIM_Cubic>
         ]
        }}}
      CurveIndex=$Emitter.NormalizedLoopAge
     )
    )
    Use Spawn Probability=true
   )
  ]
  particleSpawnScript [
   /SetVariables(Particles.ParticleVariant:NiagaraInt32=&MakeIntFromFloat(Float=&FloatFromCurve(
      FloatCurve=DataInterface() {propertyValues {Curve {keys [<at 0 0 RCIM_Constant>,<at 0.6 1 RCIM_Constant>,<at 0.8 2 RCIM_Constant>,<at 0.9 3 RCIM_Constant>,<at 0.95 4 RCIM_Constant>]}}}
      CurveIndex=&RandomRangeFloat(Minimum=0, Maximum=1)
     ))) {isSetParameters=true, module=SetVariables_#3}
   &PartitionParticles()
   &InitializeParticle(
    Color Mode=::NewEnumerator0
    Lifetime=0.3
    Sprite Size=&Multiply_Vector2D_ByFloat(Float=1, Vector2D=&RandomRangeVector2D(Minimum=<1.5 4>, Maximum=<3 6>))
    Sprite Size Mode=::NewEnumerator4
    UsePositionOffset=false
    Write Material Random=true
   )
   &SkeletalMeshLocation(
    Bone / Socket Index=$Particles.PartitionParticles.Partition
    Bone Sampling Mode=::NewEnumerator4
    MaskSpawnGroup=false
    Sampled Position Offset=&SelectVectorFromArray(
     Direct Array Index=$Particles.PartitionParticles.Partition
     Vector Selection Array=DataInterface() {propertyValues {FloatData [<3 10 2>, <3 10 -2>]}}
    )
    Skeletal Mesh=$System.Skeletal Mesh Face
   )
   &ShapeLocation(MaskSpawnGroup=false, Non Uniform Scale=<1 0 2>, Override Seed=false, Sphere Radius=3, UseOwnerScale=true)
   &DynamicMaterialParameters(
    Index 0 Param 1=&RandomRangeFloat(Minimum=0.1, Maximum=0.3)
    Index 0 Param 2=&RandomRangeFloat(Minimum=1.5, Maximum=2.5)
    Index 0 Param 3=&RandomRangeFloat(Minimum=0.1, Maximum=0.5)
    Index 0 Param 4=1
    Module.Index0Float0Write=true
    Module.Index0Float1Write=true
    Module.Index0Float2Write=true
    Module.Index0Float3Write=true
   )
   &CameraOffset(Camera Offset Amount=2)
  ]
  particleUpdateScript [
   &ParticleState()
   &ScaleSpriteSize(
    Scale Factor=&Multiply_Vector2D(
     A=&SelectVector2DFromArray(
      Direct Array Index=$Particles.ParticleVariant
      Vector 2D Selection Array=DataInterface() {propertyValues {FloatData [<1 1>, <1 1>, <0.8 1.2>, <0.4 1.5>, <0.2 2>]}}
     )
     B=&MakeVector2D(
      X=&FloatFromCurve(FloatCurve=DataInterface() {propertyValues {Curve {keys [<at 0.5 1>,<at 1 0>]}}})
      Y=&FloatFromCurve(FloatCurve=DataInterface() {propertyValues {Curve {keys [<at 0 1>,<at 1 2>]}}})
     )
    )
    Scale Sprite Size Mode=::NewEnumerator2
   )
   &ScaleSpriteSize(
    Scale Sprite Size Mode=::NewEnumerator0
    Uniform Scale Factor=&MakeCustomFloatFromBool(Bool=$User.bisHeadHidden, True Float=0, False Float=1)
   ) {module=ScaleSpriteSize001}
   &Color(Color=&SelectLinearColorFromArray(
     Array Sampling Mode=::NewEnumerator1
     Direct Array Index=$Particles.ParticleVariant
     Color Selection Array=DataInterface() {propertyValues {ColorData [<rgba 0.0087 0.6121 1 1>, <rgba 0 0.0336 1 1>, <rgba 0.4623 0 1 1>, <rgba 1 0 0.0327 1>, <rgba 1 0.7693 0 1>]}}
     Random Seed=?
    ))
   &ScaleColor(
    Color Value To Scale=$Particles.Color
    Scale Alpha=&FloatFromCurve(
     FloatCurve=DataInterface() {propertyValues {Curve {keys [
         <at 0 0 RCIM_Cubic>
         {arriveTangent=0.0087, interpMode=RCIM_Cubic, leaveTangent=0.0087, tangentMode=RCTM_User, time=0.15, value=1.0023}
         {arriveTangent=-1.1792, interpMode=RCIM_Cubic, leaveTangent=-1.1792, time=0.7, value=0.8}
         <at 1 0>
        ]}}}
     Scale Curve=10
    )
    Scale RGB=&Multiply_VectorByFloat(
     Vector=&MakeVectorFromLinearColorRGB(Color=&RandomRangeLinearColor(Minimum=<rgba 0.5938 0.6805 1 1>, Maximum=<rgba 1 0.5417 0.5422 1>))
     Float=2
    )
    ScaleA=true
    ScaleRGB=true
   )
   &SkeletalMeshLocation(
    Bone / Socket Index=$Particles.PartitionParticles.Partition
    Bone Sampling Mode=::NewEnumerator4
    Initialize Particle Position To Sampled Position=false
    MaskSpawnGroup=false
    Orientation Sampling=::NewEnumerator1
    Position Sampling=::NewEnumerator1
    Sampled Position Coordinate Space=::World
    Skeletal Mesh=$System.Skeletal Mesh Face
   ) {module=SkeletalMeshLocation001}
  ]
  rendererClasses [&NiagaraSpriteRendererProperties]
  renderers [{rendererClass=&NiagaraSpriteRendererProperties}]
 }
 GlitchSprites_LOW {
  simTarget=GPUComputeSim
  emitterUpdateScript [
   &EmitterState()
   &SpawnRate(Spawn Probability=0.3, SpawnRate=120, Use Spawn Probability=true)
  ]
  particleSpawnScript [
   /SetVariables(Particles.ParticleVariant:NiagaraInt32=&MakeIntFromFloat(Float=&FloatFromCurve(
      FloatCurve=DataInterface() {propertyValues {Curve {keys [<at 0 0 RCIM_Constant>,<at 0.6 1 RCIM_Constant>,<at 0.8 2 RCIM_Constant>,<at 0.9 3 RCIM_Constant>,<at 0.95 4 RCIM_Constant>]}}}
      CurveIndex=&RandomRangeFloat(Minimum=0, Maximum=1)
     ))) {isSetParameters=true, module=SetVariables_#3}
   &PartitionParticles()
   &InitializeParticle(
    Color Mode=::NewEnumerator0
    Lifetime=0.3
    Sprite Size=&Multiply_Vector2D_ByFloat(Float=1.5, Vector2D=&RandomRangeVector2D(Minimum=<1.5 4>, Maximum=<3 6>))
    Sprite Size Mode=::NewEnumerator4
    UsePositionOffset=false
    Write Material Random=true
   )
   &SkeletalMeshLocation(
    Bone / Socket Index=$Particles.PartitionParticles.Partition
    Bone Sampling Mode=::NewEnumerator4
    MaskSpawnGroup=false
    Sampled Position Offset=&SelectVectorFromArray(
     Direct Array Index=$Particles.PartitionParticles.Partition
     Vector Selection Array=DataInterface() {propertyValues {FloatData [<3 10 2>, <3 10 -2>]}}
    )
    Skeletal Mesh=$System.Skeletal Mesh Face
   )
   &ShapeLocation(MaskSpawnGroup=false, Non Uniform Scale=<1 0 2>, Override Seed=false, Sphere Radius=3, UseOwnerScale=true)
   &DynamicMaterialParameters(
    Index 0 Param 1=&RandomRangeFloat(Minimum=0.1, Maximum=0.3)
    Index 0 Param 2=&RandomRangeFloat(Minimum=1.5, Maximum=2.5)
    Index 0 Param 3=&RandomRangeFloat(Minimum=0.1, Maximum=0.5)
    Index 0 Param 4=1
    Module.Index0Float0Write=true
    Module.Index0Float1Write=true
    Module.Index0Float2Write=true
    Module.Index0Float3Write=true
   )
   &CameraOffset(Camera Offset Amount=2)
  ]
  particleUpdateScript [
   &ParticleState()
   &ScaleSpriteSize(
    Scale Factor=&Multiply_Vector2D(
     A=&SelectVector2DFromArray(
      Direct Array Index=$Particles.ParticleVariant
      Vector 2D Selection Array=DataInterface() {propertyValues {FloatData [<1 1>, <1 1>, <0.8 1.2>, <0.4 1.5>, <0.2 2>]}}
     )
     B=&MakeVector2D(
      X=&FloatFromCurve(FloatCurve=DataInterface() {propertyValues {Curve {keys [<at 0.5 1>,<at 1 0>]}}})
      Y=&FloatFromCurve(FloatCurve=DataInterface() {propertyValues {Curve {keys [<at 0 1>,<at 1 2>]}}})
     )
    )
    Scale Sprite Size Mode=::NewEnumerator2
   )
   &ScaleSpriteSize(
    Scale Sprite Size Mode=::NewEnumerator0
    Uniform Scale Factor=&MakeCustomFloatFromBool(Bool=$User.bisHeadHidden, True Float=0, False Float=1)
   ) {module=ScaleSpriteSize001}
   &Color(Color=&SelectLinearColorFromArray(
     Array Sampling Mode=::NewEnumerator1
     Direct Array Index=$Particles.ParticleVariant
     Color Selection Array=DataInterface() {propertyValues {ColorData [<rgba 0.0087 0.6121 1 1>, <rgba 0 0.0336 1 1>, <rgba 0.4623 0 1 1>, <rgba 1 0 0.0327 1>, <rgba 1 0.7693 0 1>]}}
     Random Seed=?
    ))
   &ScaleColor(
    Color Value To Scale=$Particles.Color
    Scale Alpha=&FloatFromCurve(
     FloatCurve=DataInterface() {propertyValues {Curve {keys [
         <at 0 0 RCIM_Cubic>
         {arriveTangent=0.0087, interpMode=RCIM_Cubic, leaveTangent=0.0087, tangentMode=RCTM_User, time=0.15, value=1.0023}
         {arriveTangent=-1.1792, interpMode=RCIM_Cubic, leaveTangent=-1.1792, time=0.7, value=0.8}
         <at 1 0>
        ]}}}
     Scale Curve=10
    )
    Scale RGB=&Multiply_VectorByFloat(
     Vector=&MakeVectorFromLinearColorRGB(Color=&RandomRangeLinearColor(Minimum=<rgba 0.5938 0.6805 1 1>, Maximum=<rgba 1 0.5417 0.5422 1>))
     Float=3
    )
    ScaleA=true
    ScaleRGB=true
   )
   &SkeletalMeshLocation(
    Bone / Socket Index=$Particles.PartitionParticles.Partition
    Bone Sampling Mode=::NewEnumerator4
    Initialize Particle Position To Sampled Position=false
    MaskSpawnGroup=false
    Orientation Sampling=::NewEnumerator1
    Position Sampling=::NewEnumerator1
    Sampled Position Coordinate Space=::World
    Skeletal Mesh=$System.Skeletal Mesh Face
   ) {module=SkeletalMeshLocation001}
  ]
  rendererClasses [&NiagaraSpriteRendererProperties]
  renderers [{rendererClass=&NiagaraSpriteRendererProperties}]
 }
 Ribbon {
  emitterUpdateScript [
   /SetVariables(Emitter.CountPerRibbon=7) {isSetParameters=true}
   &EmitterState()
   &SpawnBurst_Instantaneous(
    Spawn Count=&Multiply_Int(A=$Emitter.CountPerRibbon, B=2)
    Use Spawn Probability=false
    UseLoopCountLimit=false
   )
  ]
  particleSpawnScript [
   &FN_SetSpawnGroup(
    Inclusivity=true
    Normalize Spawn Group=true
    Particles Per Group:NiagaraInt32=$Emitter.CountPerRibbon
    Pattern=::NewEnumerator0
    Set Ribbon ID and Ribbon Link Order=true
   )
   &InitializeParticle(
    Lifetime=$User.SystemDuration
    Ribbon Width=2
    Ribbon Width Mode=::NewEnumerator1
    Sprite Size Mode=::NewEnumerator3
    UsePositionOffset=false
    Write Material Random=true
   )
   &CameraOffset()
   /Game/Effects/Niagara/Modules/BoneFlow/FN_BoneFlow_v4/FN_BoneFlow(
    BoneID:NiagaraInt32=$Particles.SpawnGroup
    Height:NiagaraFloat=$Particles.Normalized Spawn Group
    SkeletalMesh:NiagaraDataInterfaceSkeletalMesh=$System.Skeletal Mesh Ribbon
    Use Sockets Instead=true
    Write Position Attribute=true
   )
  ]
  particleUpdateScript [
   &ParticleState()
   &FN_BoneFlow(
    BoneID:NiagaraInt32=$Particles.SpawnGroup
    Height:NiagaraFloat=$Particles.Normalized Spawn Group
    SkeletalMesh:NiagaraDataInterfaceSkeletalMesh=$System.Skeletal Mesh Ribbon
    Use Sockets Instead=true
    Write Position Attribute=true
   ) {module=FN_BoneFlow001}
   /SetVariables(Particles.PositionOffset:NiagaraPosition=&AddVectorToPosition(
     Position=$Particles.Position
     Vector=&Multiply_VectorByFloat(
      Vector=&ReturnCameraVector(
       Camera Vectors=::NewEnumerator4
       Coordinate Space=::Simulation
       Origin Position=$Particles.Position
       CameraQuery=DataInterface() {propertyValues {bRequireCurrentFrameData=false}}
      )
      Float=10
     )
    )) {isSetParameters=true}
   &ScaleColor(
    Scale Alpha=&Multiply_Float(
     A=&Multiply_Float(
      A=&FloatFromCurve(
       FloatCurve=DataInterface() {propertyValues {Curve {
          defaultValue=1
          keys [
           <at 0 0 RCIM_Cubic>
           {arriveTangentWeight=0.1389, interpMode=RCIM_Cubic, leaveTangentWeight=0.2778, tangentMode=RCTM_Break, time=0.25, value=1}
           {arriveTangentWeight=0.2778, interpMode=RCIM_Cubic, leaveTangentWeight=0.1389, tangentMode=RCTM_Break, time=0.75, value=1}
           <at 1 0 RCIM_Cubic>
          ]
         }}}
       Scale Curve=30
      )
      B=&Lerp_Float(
       A=&ClampFloat(Float=&Waveform("Waveform [1]"=::NewEnumerator8, "[1] Frequency"=20, "[1] Amplitude Min/Max"=<0 1>, "[1] Exponent"=2))
       B=1
       Alpha=&FloatFromCurve(FloatCurve=DataInterface() {propertyValues {Curve {
           defaultValue=1
           keys [
            <at 0 0 RCIM_Cubic>
            {arriveTangentWeight=0.1389, interpMode=RCIM_Cubic, leaveTangentWeight=0.2778, tangentMode=RCTM_Break, time=0.4, value=1}
            {arriveTangentWeight=0.2778, interpMode=RCIM_Cubic, leaveTangentWeight=0.1389, tangentMode=RCTM_Break, time=0.6, value=1}
            <at 1 0 RCIM_Cubic>
           ]
          }}})
      )
     )
     B=&Lerp_Float(
      A=0.25
      B=1
      Alpha=&Waveform(
       "Waveform [1]"=::NewEnumerator8
       "[1] Frequency"=8
       "[1] Amplitude Min/Max"=<0 1>
       "[1] Exponent"=2
       "[1] Random Seed"=$Particles.Initial.SpawnGroup
      )
     )
    )
    Scale RGB=&FN_Select_Vector_from_Time_Of_Day(Morning=<1 1 1>, Day=<1 1 1>, Evening=<1 1 1>, Night=<3 3 3>)
    ScaleA=true
    ScaleRGB=true
   )
   &DynamicMaterialParameters(
    Index 0 Param 1=&RandomRangeFloat(Minimum=0, Maximum=1, Randomness Mode=::NewEnumerator1, Random Seed=$Particles.SpawnGroup, Fixed Random Seed=true)
    Index 0 Param 2=&RandomRangeFloat(Minimum=0, Maximum=1, Randomness Mode=::NewEnumerator1, Random Seed=$Particles.SpawnGroup, Fixed Random Seed=true)
    Index 0 Param 3=&Waveform("Waveform [1]"=::NewEnumerator8, "[1] Frequency"=2, "[1] Amplitude Min/Max"=<0 1>, "[1] Random Seed"=$Particles.Initial.SpawnGroup)
    Module.Index0Float0Write=true
    Module.Index0Float1Write=true
    Module.Index0Float2Write=true
    Module.Index0Float3Write=true
   )
   &ScaleRibbonWidth(Ribbon Width Scale=&MakeCustomFloatFromBool(Bool=$User.bisHeadHidden, True Float=0, False Float=1))
  ]
  rendererClasses [&NiagaraRibbonRendererProperties]
  renderers [{rendererClass=&NiagaraRibbonRendererProperties}]
 }
}
summary {
 systemName=NS_Character_PawJasmineGravityGreet_Reactivity_HoodDown
 emitters [
  {emitterName=GlitchSprites, simTarget=GPUComputeSim, rendererClasses [&NiagaraSpriteRendererProperties]}
  {emitterName=Ribbon, rendererClasses [&NiagaraRibbonRendererProperties]}
  {emitterName=GlitchSprites_LOW, simTarget=GPUComputeSim, rendererClasses [&NiagaraSpriteRendererProperties]}
 ]
 userVariables [
  User.bisHeadHidden {type=NiagaraBool}
  User.SystemDuration {type=NiagaraFloat, defaultValue=3, description=""}
 ]
}
