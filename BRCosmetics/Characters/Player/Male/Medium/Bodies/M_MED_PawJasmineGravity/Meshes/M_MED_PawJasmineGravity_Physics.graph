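# Physics asset for the PawJasmineGravity medium male player cosmetic body, setting up a physics chain on the arm.
# Two three-link capsule chains hang from the left and right MSTR_armChain bodies through swing- and twist-limited constraints.
# An extra constraint joins the two chain tips, closing them into one linked span between both roots.
# Unconstrained capsules on the right hand, forearm twist and thumb bones appear to serve as colliders for the chain.
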PhysicsAsset /BRCosmetics/Characters/Player/Male/Medium/Bodies/M_MED_PawJasmineGravity/Meshes/M_MED_PawJasmineGravity_Physics
bodies {
 MSTR_armChain_l [{shapeType=Capsule, length=0.505, radius=0.505, shapeName=MSTR_armChain_l_capsule, rotation=<rot 0 -90 0>}]
 MSTR_armChain_r [{shapeType=Capsule, length=0.505, radius=0.505, shapeName=MSTR_armChain_r_capsule, rotation=<rot 0 -90 0>}]
 armChain_l_1 [{shapeType=Capsule, length=0.5814, radius=1.2817, shapeName=armChain_l_1_capsule, center=<1.9069 0 -0.5089>, rotation=<rot 90 0 0>}]
 armChain_l_2 [{shapeType=Capsule, length=0.8853, radius=1.4546, shapeName=armChain_l_2_capsule, center=<1.8791 0 -0.3393>, rotation=<rot 90 0 0>}]
 armChain_l_3 [{shapeType=Capsule, length=1.1235, radius=1.19, shapeName=armChain_l_3_capsule, center=<1.5393 0 -0.229>, rotation=<rot 90 0 0>}]
 armChain_r_1 [{
   shapeType=Capsule
   length=1.3751
   radius=1.0962
   shapeName=armChain_r_1_capsule
   center=<1.4882 0.1183 0.6002>
   rotation=<rot -87.1877 -0.0162 0.016>
  }]
 armChain_r_2 [{
   shapeType=Capsule
   length=1.4745
   radius=1.0546
   shapeName=armChain_r_2_capsule
   center=<1.1767 0 0.3496>
   rotation=<rot -87.1876 -0.0087 0.008>
  }]
 armChain_r_3 [{
   shapeType=Capsule
   length=1.0374
   radius=1.0974
   shapeName=armChain_r_3_capsule
   center=<1.4983 0 0.3945>
   rotation=<rot -87.1875 -0.0014 0.0006>
  }]
 hand_r [{
   shapeType=Capsule
   length=4.3465
   radius=5.4758
   shapeName=hand_r_capsule
   center=<-5.6424 3.1027 -0.0281>
   rotation=<rot -87.1877 -2.8099 2.8099>
  }]
 lowerarm_twist_01_r [{
   shapeType=Capsule
   length=4.0585
   radius=6.4493
   shapeName=lowerarm_twist_01_r_capsule
   center=<3.307 -0.4339 -1.2567>
   rotation=<rot 75.9376 -0.0005 -0.001>
  }]
 thumb_01_r [{
   shapeType=Capsule
   length=3.7033
   radius=2.6118
   shapeName=thumb_01_r_capsule
   center=<-2.5987 0.4294 0.1717>
   rotation=<rot -84.3748 0 0.0014>
  }]
}
constraints [
 {bone1Name=armChain_l_3, bone2Name=armChain_l_2, swing1LimitDegrees=30, swing2LimitDegrees=30, twistLimitDegrees=5}
 {bone1Name=armChain_l_2, bone2Name=armChain_l_1, swing1LimitDegrees=40, swing2LimitDegrees=40, twistLimitDegrees=5}
 {bone1Name=armChain_l_1, bone2Name=MSTR_armChain_l, swing1LimitDegrees=65, swing2LimitDegrees=65, twistLimitDegrees=5}
 {bone1Name=armChain_r_3, bone2Name=armChain_r_2, swing1LimitDegrees=30, swing2LimitDegrees=30, twistLimitDegrees=5}
 {bone1Name=armChain_r_2, bone2Name=armChain_r_1, swing1LimitDegrees=30, swing2LimitDegrees=30, twistLimitDegrees=5}
 {bone1Name=armChain_r_1, bone2Name=MSTR_armChain_r, swing1LimitDegrees=30, swing2LimitDegrees=30, twistLimitDegrees=5}
 {bone1Name=armChain_l_3, bone2Name=armChain_r_3, swing1LimitDegrees=45, swing2LimitDegrees=45, twistLimitDegrees=5}
]
