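# Physics asset for the face accessory part of the male medium PawJasmineGravity character body, covering its open-cap bones.
# A small capsule on MSTR_cap_open and a larger, offset zero-length capsule, effectively a sphere, on cap_open form the two bodies.
# One constraint joins them, with narrow twist and swing1 limits and a much wider swing2 limit, favoring a hinge-like swing on one axis.
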PhysicsAsset /BRCosmetics/Characters/Player/Male/Medium/Bodies/M_MED_PawJasmineGravity/Meshes/Parts/M_MED_PawJasmineGravity_FaceAcc_Physics
bodies {
 MSTR_cap_open [{shapeType=Capsule, length=0.505, radius=0.505, shapeName=MSTR_cap_open_capsule, rotation=<rot 0 -90 0>}]
 cap_open [{
   shapeType=Capsule
   length=0
   radius=3.9881
   shapeName=cap_open_capsule
   center=<7.2798 -0.0259 0.6056>
   rotation=<rot 3.2858 90.4141 -90.0004>
  }]
}
constraints [{bone1Name=MSTR_cap_open, bone2Name=cap_open, swing1LimitDegrees=2, swing2LimitDegrees=35, twistLimitDegrees=1}]
