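# Physics asset for the hair on the face accessory of the male medium PawJasmineGravity player cosmetic body.
# Seventeen strands (A–Q) each start at a small MSTR capsule, followed by one or two larger, offset capsules on the hair bones.
# Constraints chain each segment to its parent with identical limits: a wider first swing axis, narrow second swing and tight twist, so strands bend mostly along one axis.
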PhysicsAsset /BRCosmetics/Characters/Player/Male/Medium/Bodies/M_MED_PawJasmineGravity/Meshes/Parts/M_MED_PawJasmineGravity_FaceAcc_Physics_Hair
bodies {
 MSTR_hairA [{shapeType=Capsule, length=0.2, radius=0.2, shapeName=MSTR_hairA_capsule}]
 MSTR_hairB [{shapeType=Capsule, length=0.2, radius=0.2, shapeName=MSTR_hairB_capsule}]
 MSTR_hairC [{shapeType=Capsule, length=0.2, radius=0.2, shapeName=MSTR_hairC_capsule}]
 MSTR_hairD [{shapeType=Capsule, length=0.2, radius=0.2, shapeName=MSTR_hairD_capsule}]
 MSTR_hairE [{shapeType=Capsule, length=0.2, radius=0.2, shapeName=MSTR_hairE_capsule}]
 MSTR_hairF [{shapeType=Capsule, length=0.2, radius=0.2, shapeName=MSTR_hairF_capsule}]
 MSTR_hairG [{shapeType=Capsule, length=0.2, radius=0.2, shapeName=MSTR_hairG_capsule}]
 MSTR_hairH [{shapeType=Capsule, length=0.2, radius=0.2, shapeName=MSTR_hairH_capsule}]
 MSTR_hairI [{shapeType=Capsule, length=0.2, radius=0.2, shapeName=MSTR_hairI_capsule}]
 MSTR_hairJ [{shapeType=Capsule, length=0.2, radius=0.2, shapeName=MSTR_hairJ_capsule}]
 MSTR_hairK [{shapeType=Capsule, length=0.2, radius=0.2, shapeName=MSTR_hairK_capsule}]
 MSTR_hairL [{shapeType=Capsule, length=0.2, radius=0.2, shapeName=MSTR_hairL_capsule}]
 MSTR_hairM [{shapeType=Capsule, length=0.2, radius=0.2, shapeName=MSTR_hairM_capsule}]
 MSTR_hairN [{shapeType=Capsule, length=0.2, radius=0.2, shapeName=MSTR_hairN_capsule}]
 MSTR_hairO [{shapeType=Capsule, length=0.2, radius=0.2, shapeName=MSTR_hairO_capsule}]
 MSTR_hairP [{shapeType=Capsule, length=0.2, radius=0.2, shapeName=MSTR_hairP_capsule}]
 MSTR_hairQ [{shapeType=Capsule, length=0.2, radius=0.2, shapeName=MSTR_hairQ_capsule}]
 hairA_1 [{shapeType=Capsule, length=0.5, radius=0.5, shapeName=hairA_1_capsule, center=<0.9368 0 0>}]
 hairA_2 [{shapeType=Capsule, length=0.5, radius=0.5, shapeName=hairA_2_capsule, center=<0.9368 0 0>}]
 hairB_1 [{shapeType=Capsule, length=0.5, radius=0.5, shapeName=hairB_1_capsule, center=<0.9368 0 0>}]
 hairB_2 [{shapeType=Capsule, length=0.5, radius=0.5, shapeName=hairB_2_capsule, center=<0.9368 0 0>}]
 hairC_1 [{shapeType=Capsule, length=0.5, radius=0.5, shapeName=hairC_1_capsule, center=<0.9368 0 0>}]
 hairC_2 [{shapeType=Capsule, length=0.5, radius=0.5, shapeName=hairC_2_capsule, center=<0.9368 0 0>}]
 hairD_1 [{shapeType=Capsule, length=0.5, radius=0.5, shapeName=hairD_1_capsule, center=<0.9368 0 0>}]
 hairE_1 [{shapeType=Capsule, length=0.5, radius=0.5, shapeName=hairE_1_capsule, center=<0.9368 0 0>}]
 hairF_1 [{shapeType=Capsule, length=0.5, radius=0.5, shapeName=hairF_1_capsule, center=<0.9368 0 0>}]
 hairG_1 [{shapeType=Capsule, length=0.5, radius=0.5, shapeName=hairG_1_capsule, center=<0.9368 0 0>}]
 hairH_1 [{shapeType=Capsule, length=0.5, radius=0.5, shapeName=hairH_1_capsule, center=<0.9368 0 0>}]
 hairH_2 [{shapeType=Capsule, length=0.5, radius=0.5, shapeName=hairH_2_capsule, center=<0.9368 0 0>}]
 hairI_1 [{shapeType=Capsule, length=0.5, radius=0.5, shapeName=hairI_1_capsule, center=<0.9368 0 0>}]
 hairI_2 [{shapeType=Capsule, length=0.5, radius=0.5, shapeName=hairI_2_capsule, center=<0.9368 0 0>}]
 hairJ_1 [{shapeType=Capsule, length=0.5, radius=0.5, shapeName=hairJ_1_capsule, center=<0.9368 0 0>}]
 hairJ_2 [{shapeType=Capsule, length=0.5, radius=0.5, shapeName=hairJ_2_capsule, center=<0.9368 0 0>}]
 hairK_1 [{shapeType=Capsule, length=0.5, radius=0.5, shapeName=hairK_1_capsule, center=<0.9368 0 0>}]
 hairK_2 [{shapeType=Capsule, length=0.5, radius=0.5, shapeName=hairK_2_capsule, center=<0.9368 0 0>}]
 hairL_1 [{shapeType=Capsule, length=0.5, radius=0.5, shapeName=hairL_1_capsule, center=<0.9368 0 0>}]
 hairL_2 [{shapeType=Capsule, length=0.5, radius=0.5, shapeName=hairL_2_capsule, center=<0.9368 0 0>}]
 hairM_1 [{shapeType=Capsule, length=0.5, radius=0.5, shapeName=hairM_1_capsule, center=<0.9368 0 0>}]
 hairM_2 [{shapeType=Capsule, length=0.5, radius=0.5, shapeName=hairM_2_capsule, center=<0.9368 0 0>}]
 hairN_1 [{shapeType=Capsule, length=0.5, radius=0.5, shapeName=hairN_1_capsule, center=<0.9368 0 0>}]
 hairO_1 [{shapeType=Capsule, length=0.5, radius=0.5, shapeName=hairO_1_capsule, center=<0.9368 0 0>}]
 hairP_1 [{shapeType=Capsule, length=0.5, radius=0.5, shapeName=hairP_1_capsule, center=<0.9368 0 0>}]
 hairP_2 [{shapeType=Capsule, length=0.5, radius=0.5, shapeName=hairP_2_capsule, center=<0.9368 0 0>}]
 hairQ_1 [{shapeType=Capsule, length=0.5, radius=0.5, shapeName=hairQ_1_capsule, center=<0.9368 0 0>}]
 hairQ_2 [{shapeType=Capsule, length=0.5, radius=0.5, shapeName=hairQ_2_capsule, center=<0.9368 0 0>}]
}
constraints [
 {bone1Name=hairQ_2, bone2Name=hairQ_1, swing1LimitDegrees=25, swing2LimitDegrees=5, twistLimitDegrees=3}
 {bone1Name=hairQ_1, bone2Name=MSTR_hairQ, swing1LimitDegrees=25, swing2LimitDegrees=5, twistLimitDegrees=3}
 {bone1Name=hairP_2, bone2Name=hairP_1, swing1LimitDegrees=25, swing2LimitDegrees=5, twistLimitDegrees=3}
 {bone1Name=hairP_1, bone2Name=MSTR_hairP, swing1LimitDegrees=25, swing2LimitDegrees=5, twistLimitDegrees=3}
 {bone1Name=hairO_1, bone2Name=MSTR_hairO, swing1LimitDegrees=25, swing2LimitDegrees=5, twistLimitDegrees=3}
 {bone1Name=hairN_1, bone2Name=MSTR_hairN, swing1LimitDegrees=25, swing2LimitDegrees=5, twistLimitDegrees=3}
 {bone1Name=hairM_2, bone2Name=hairM_1, swing1LimitDegrees=25, swing2LimitDegrees=5, twistLimitDegrees=3}
 {bone1Name=hairM_1, bone2Name=MSTR_hairM, swing1LimitDegrees=25, swing2LimitDegrees=5, twistLimitDegrees=3}
 {bone1Name=hairL_2, bone2Name=hairL_1, swing1LimitDegrees=25, swing2LimitDegrees=5, twistLimitDegrees=3}
 {bone1Name=hairL_1, bone2Name=MSTR_hairL, swing1LimitDegrees=25, swing2LimitDegrees=5, twistLimitDegrees=3}
 {bone1Name=hairK_2, bone2Name=hairK_1, swing1LimitDegrees=25, swing2LimitDegrees=5, twistLimitDegrees=3}
 {bone1Name=hairK_1, bone2Name=MSTR_hairK, swing1LimitDegrees=25, swing2LimitDegrees=5, twistLimitDegrees=3}
 {bone1Name=hairJ_2, bone2Name=hairJ_1, swing1LimitDegrees=25, swing2LimitDegrees=5, twistLimitDegrees=3}
 {bone1Name=hairJ_1, bone2Name=MSTR_hairJ, swing1LimitDegrees=25, swing2LimitDegrees=5, twistLimitDegrees=3}
 {bone1Name=hairI_2, bone2Name=hairI_1, swing1LimitDegrees=25, swing2LimitDegrees=5, twistLimitDegrees=3}
 {bone1Name=hairI_1, bone2Name=MSTR_hairI, swing1LimitDegrees=25, swing2LimitDegrees=5, twistLimitDegrees=3}
 {bone1Name=hairH_2, bone2Name=hairH_1, swing1LimitDegrees=25, swing2LimitDegrees=5, twistLimitDegrees=3}
 {bone1Name=hairH_1, bone2Name=MSTR_hairH, swing1LimitDegrees=25, swing2LimitDegrees=5, twistLimitDegrees=3}
 {bone1Name=hairG_1, bone2Name=MSTR_hairG, swing1LimitDegrees=25, swing2LimitDegrees=5, twistLimitDegrees=3}
 {bone1Name=hairF_1, bone2Name=MSTR_hairF, swing1LimitDegrees=25, swing2LimitDegrees=5, twistLimitDegrees=3}
 {bone1Name=hairE_1, bone2Name=MSTR_hairE, swing1LimitDegrees=25, swing2LimitDegrees=5, twistLimitDegrees=3}
 {bone1Name=hairD_1, bone2Name=MSTR_hairD, swing1LimitDegrees=25, swing2LimitDegrees=5, twistLimitDegrees=3}
 {bone1Name=hairC_2, bone2Name=hairC_1, swing1LimitDegrees=25, swing2LimitDegrees=5, twistLimitDegrees=3}
 {bone1Name=hairC_1, bone2Name=MSTR_hairC, swing1LimitDegrees=25, swing2LimitDegrees=5, twistLimitDegrees=3}
 {bone1Name=hairB_2, bone2Name=hairB_1, swing1LimitDegrees=25, swing2LimitDegrees=5, twistLimitDegrees=3}
 {bone1Name=hairB_1, bone2Name=MSTR_hairB, swing1LimitDegrees=25, swing2LimitDegrees=5, twistLimitDegrees=3}
 {bone1Name=hairA_2, bone2Name=hairA_1, swing1LimitDegrees=25, swing2LimitDegrees=5, twistLimitDegrees=3}
 {bone1Name=hairA_1, bone2Name=MSTR_hairA, swing1LimitDegrees=25, swing2LimitDegrees=5, twistLimitDegrees=3}
]
