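# Physics asset for the hood on the face accessory part of the male medium PawJasmineGravity player cosmetic.
# It gives the left and right upper cap side bones a capsule body each, alongside smaller capsules on their MSTR counterpart bones.
# Each side bone is constrained to its MSTR bone with twist and second swing locked, so the first swing is the only unlocked angular axis.
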PhysicsAsset /BRCosmetics/Characters/Player/Male/Medium/Bodies/M_MED_PawJasmineGravity/Meshes/Parts/M_MED_PawJasmineGravity_FaceAcc_Physics_Hood
bodies {
 MSTR_cap_up_side_l [{shapeType=Capsule, length=0.505, radius=0.505, shapeName=MSTR_cap_up_side_l_capsule}]
 MSTR_cap_up_side_r [{shapeType=Capsule, length=0.505, radius=0.505, shapeName=MSTR_cap_up_side_r_capsule}]
 cap_up_side_l [{shapeType=Capsule, length=1, radius=1, shapeName=cap_up_side_l_capsule}]
 cap_up_side_r [{shapeType=Capsule, length=1, radius=1, shapeName=cap_up_side_r_capsule}]
}
constraints [
 {
  bone1Name=cap_up_side_r
  bone2Name=MSTR_cap_up_side_r
  swing1LimitDegrees=1
  swing2LimitDegrees=1
  swing2Motion=Locked
  twistLimitDegrees=1
  twistMotion=Locked
 }
 {
  bone1Name=cap_up_side_l
  bone2Name=MSTR_cap_up_side_l
  swing1LimitDegrees=1
  swing2LimitDegrees=1
  swing2Motion=Locked
  twistLimitDegrees=1
  twistMotion=Locked
 }
]
